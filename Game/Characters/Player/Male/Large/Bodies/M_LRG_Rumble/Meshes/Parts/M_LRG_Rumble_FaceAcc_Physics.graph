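# Physics asset for the spiky hair on the Rumble large male character's FaceAcc mesh, using capsule bodies per bone.
# Lettered spike chains (A–X) each run from a small master body on the head through one to four segments.
# Swing limits tighten beyond each master, with zero twist except on chain Q.
# Head links freely to spine_05's two large torso capsules; free constraints also pair tips S3–W3 and Q3–G4.
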PhysicsAsset /Game/Characters/Player/Male/Large/Bodies/M_LRG_Rumble/Meshes/Parts/M_LRG_Rumble_FaceAcc_Physics
bodies {
 HairSpikesA1 [{shapeType=Capsule, length=5, radius=3, shapeName=HairSpikesA1_capsule, center=<5 0 0>, rotation=<rot 90 -6.7214 -6.7214>}]
 HairSpikesA2 [{shapeType=Capsule, length=5, radius=3, shapeName=HairSpikesA2_capsule, center=<5 0 0>, rotation=<rot 90 -172.4476 -172.4476>}]
 HairSpikesA3 [{shapeType=Capsule, length=5, radius=3, shapeName=HairSpikesA3_capsule, center=<5 0 0>, rotation=<rot 90 -179.9703 -179.9703>}]
 HairSpikesA_MSTR [{shapeType=Capsule, length=1, radius=1, shapeName=HairSpikesA_MSTR_capsule}]
 HairSpikesB1 [{shapeType=Capsule, length=5.5, radius=2.5, shapeName=HairSpikesB1_capsule, center=<10 0 0>, rotation=<rot 90 -179.9473 -179.9473>}]
 HairSpikesB2 [{shapeType=Capsule, length=5.5, radius=2.5, shapeName=HairSpikesB2_capsule, center=<10 0 0>, rotation=<rot 90 -179.2404 -179.2404>}]
 HairSpikesB3 [{shapeType=Capsule, length=4.25, radius=3.75, shapeName=HairSpikesB3_capsule, center=<10 0 0>, rotation=<rot 90 -90 -90>}]
 HairSpikesB_MSTR [{shapeType=Capsule, length=1, radius=1, shapeName=HairSpikesB_MSTR_capsule}]
 HairSpikesC1 [{shapeType=Capsule, length=5, radius=3, shapeName=HairSpikesC1_capsule, center=<5 0 0>, rotation=<rot 90 -179.9802 -179.9802>}]
 HairSpikesC2 [{shapeType=Capsule, length=5, radius=3, shapeName=HairSpikesC2_capsule, center=<5 0 0>, rotation=<rot 90 -179.9604 -179.9604>}]
 HairSpikesC3 [{shapeType=Capsule, length=5, radius=3, shapeName=HairSpikesC3_capsule, center=<5 0 0>, rotation=<rot 90 -179.8417 -179.8417>}]
 HairSpikesC_MSTR [{shapeType=Capsule, length=1, radius=1, shapeName=HairSpikesC_MSTR_capsule}]
 HairSpikesD1 [{shapeType=Capsule, length=5, radius=3, shapeName=HairSpikesD1_capsule, center=<5 0 0>, rotation=<rot 90 -179.789 -179.789>}]
 HairSpikesD2 [{shapeType=Capsule, length=5, radius=3, shapeName=HairSpikesD2_capsule, center=<5 0 0>, rotation=<rot 90 -79.2198 -79.2198>}]
 HairSpikesD_MSTR [{shapeType=Capsule, length=1, radius=1, shapeName=HairSpikesD_MSTR_capsule}]
 HairSpikesE1 [{shapeType=Capsule, length=5, radius=3, shapeName=HairSpikesE1_capsule, center=<5 0 0>, rotation=<rot 90 -179.7468 -179.7468>}]
 HairSpikesE_MSTR [{shapeType=Capsule, length=1, radius=1, shapeName=HairSpikesE_MSTR_capsule}]
 HairSpikesF1 [{shapeType=Capsule, length=5, radius=3, shapeName=HairSpikesF1_capsule, center=<5 0 0>, rotation=<rot 90 -10.025 -10.025>}]
 HairSpikesF2 [{shapeType=Capsule, length=5, radius=3, shapeName=HairSpikesF2_capsule, center=<5 0 0>, rotation=<rot 90 -90 -90>}]
 HairSpikesF3 [{shapeType=Capsule, length=5, radius=3, shapeName=HairSpikesF3_capsule, center=<5 0 0>, rotation=<rot 90 179.8417 179.8418>}]
 HairSpikesF_MSTR [{shapeType=Capsule, length=1, radius=1, shapeName=HairSpikesF_MSTR_capsule}]
 HairSpikesG1 [{shapeType=Capsule, length=3.75, radius=4.25, shapeName=HairSpikesG1_capsule, center=<5 0 0>, rotation=<rot 90 179.9703 179.9703>}]
 HairSpikesG2 [{shapeType=Capsule, length=2.75, radius=5.25, shapeName=HairSpikesG2_capsule, center=<5 0 0>, rotation=<rot 90 179.9951 179.9951>}]
 HairSpikesG3 [{shapeType=Capsule, length=4.25, radius=3.75, shapeName=HairSpikesG3_capsule, center=<5 0 0>, rotation=<rot 90 179.7626 179.7626>}]
 HairSpikesG4 [{shapeType=Capsule, length=3.75, radius=4.25, shapeName=HairSpikesG4_capsule, center=<5 0 0>, rotation=<rot 90 179.9951 179.9951>}]
 HairSpikesG_MSTR [{shapeType=Capsule, length=1, radius=1, shapeName=HairSpikesG_MSTR_capsule}]
 HairSpikesH1 [{shapeType=Capsule, length=5, radius=3, shapeName=HairSpikesH1_capsule, center=<5 0 0>, rotation=<rot 90 -166.7373 -166.7373>}]
 HairSpikesH2 [{shapeType=Capsule, length=5, radius=3, shapeName=HairSpikesH2_capsule, center=<5 0 0>, rotation=<rot 90 -174.9489 -174.9488>}]
 HairSpikesH3 [{shapeType=Capsule, length=5, radius=3, shapeName=HairSpikesH3_capsule, center=<5 0 0>, rotation=<rot 90 10.1464 10.1464>}]
 HairSpikesH_MSTR [{shapeType=Capsule, length=1, radius=1, shapeName=HairSpikesH_MSTR_capsule}]
 HairSpikesI1 [{shapeType=Capsule, length=5, radius=3, shapeName=HairSpikesI1_capsule, center=<5 0 0>, rotation=<rot 90 -179.9852 -179.9852>}]
 HairSpikesI2 [{shapeType=Capsule, length=5, radius=3, shapeName=HairSpikesI2_capsule, center=<5 0 0>, rotation=<rot 90 -0.2374 -0.2374>}]
 HairSpikesI3 [{shapeType=Capsule, length=5, radius=3, shapeName=HairSpikesI3_capsule, center=<5 0 0>, rotation=<rot 90 0 0>}]
 HairSpikesI_MSTR [{shapeType=Capsule, length=1, radius=1, shapeName=HairSpikesI_MSTR_capsule}]
 HairSpikesJ1 [{shapeType=Capsule, length=5, radius=3, shapeName=HairSpikesJ1_capsule, center=<5 0 0>, rotation=<rot 90 -179.8417 -179.8417>}]
 HairSpikesJ2 [{shapeType=Capsule, length=5, radius=3, shapeName=HairSpikesJ2_capsule, center=<5 0 0>, rotation=<rot 90 -14.6948 -14.6948>}]
 HairSpikesJ_MSTR [{shapeType=Capsule, length=1, radius=1, shapeName=HairSpikesJ_MSTR_capsule}]
 HairSpikesK1 [{shapeType=Capsule, length=5, radius=3, shapeName=HairSpikesK1_capsule, center=<5 0 0>, rotation=<rot 90 166.7373 166.7373>}]
 HairSpikesK2 [{shapeType=Capsule, length=5, radius=3, shapeName=HairSpikesK2_capsule, center=<5 0 0>, rotation=<rot 90 5.0512 5.0512>}]
 HairSpikesK3 [{shapeType=Capsule, length=5, radius=3, shapeName=HairSpikesK3_capsule, center=<5 0 0>, rotation=<rot 90 -164.8136 -164.8136>}]
 HairSpikesK_MSTR [{shapeType=Capsule, length=1, radius=1, shapeName=HairSpikesK_MSTR_capsule}]
 HairSpikesL1 [{shapeType=Capsule, length=5, radius=3, shapeName=HairSpikesL1_capsule, center=<5 0 0>, rotation=<rot 90 -13.0767 -13.0767>}]
 HairSpikesL_MSTR [{shapeType=Capsule, length=1, radius=1, shapeName=HairSpikesL_MSTR_capsule}]
 HairSpikesM1 [{shapeType=Capsule, length=5, radius=3, shapeName=HairSpikesM1_capsule, center=<5 0 0>, rotation=<rot 90 -179.367 -179.367>}]
 HairSpikesM_MSTR [{shapeType=Capsule, length=1, radius=1, shapeName=HairSpikesM_MSTR_capsule}]
 HairSpikesN1 [{shapeType=Capsule, length=5, radius=3, shapeName=HairSpikesN1_capsule, center=<5 0 0>, rotation=<rot 90 -179.9975 -179.9975>}]
 HairSpikesN2 [{shapeType=Capsule, length=5, radius=3, shapeName=HairSpikesN2_capsule, center=<5 0 0>, rotation=<rot 90 -34.071 -34.071>}]
 HairSpikesN_MSTR [{shapeType=Capsule, length=1, radius=1, shapeName=HairSpikesN_MSTR_capsule}]
 HairSpikesO1 [{shapeType=Capsule, length=5, radius=3, shapeName=HairSpikesO1_capsule, center=<5 0 0>, rotation=<rot 90 179.9209 179.9209>}]
 HairSpikesO2 [{shapeType=Capsule, length=5, radius=3, shapeName=HairSpikesO2_capsule, center=<5 0 0>, rotation=<rot 90 130.1151 130.1151>}]
 HairSpikesO_MSTR [{shapeType=Capsule, length=1, radius=1, shapeName=HairSpikesO_MSTR_capsule}]
 HairSpikesP1 [{shapeType=Capsule, length=5, radius=3, shapeName=HairSpikesP1_capsule, center=<5 0 0>, rotation=<rot 90 179.9406 179.9407>}]
 HairSpikesP_MSTR [{shapeType=Capsule, length=1, radius=1, shapeName=HairSpikesP_MSTR_capsule}]
 HairSpikesQ1 [{shapeType=Capsule, length=4.5, radius=3.5, shapeName=HairSpikesQ1_capsule, center=<8 0 0>, rotation=<rot 90 179.9901 179.9901>}]
 HairSpikesQ2 [{shapeType=Capsule, length=4.5, radius=3.5, shapeName=HairSpikesQ2_capsule, center=<8 0 0>, rotation=<rot 90 -177.7799 -172.7797>}]
 HairSpikesQ3 [{shapeType=Capsule, length=4, radius=4, shapeName=HairSpikesQ3_capsule, center=<8 0 0>, rotation=<rot 90 -156.4922 -146.4921>}]
 HairSpikesQ_MSTR [{shapeType=Capsule, length=1, radius=1, shapeName=HairSpikesQ_MSTR_capsule}]
 HairSpikesR1 [{shapeType=Capsule, length=5, radius=3, shapeName=HairSpikesR1_capsule, center=<5 0 0>, rotation=<rot 90 179.9802 179.9802>}]
 HairSpikesR2 [{shapeType=Capsule, length=5, radius=3, shapeName=HairSpikesR2_capsule, center=<5 0 0>, rotation=<rot 90 0.0791 0.0791>}]
 HairSpikesR3 [{shapeType=Capsule, length=5, radius=3, shapeName=HairSpikesR3_capsule, center=<5 0 0>, rotation=<rot 90 -179.9975 -179.9975>}]
 HairSpikesR_MSTR [{shapeType=Capsule, length=1, radius=1, shapeName=HairSpikesR_MSTR_capsule}]
 HairSpikesS1 [{shapeType=Capsule, length=5, radius=3, shapeName=HairSpikesS1_capsule, center=<5 0 0>, rotation=<rot 90 172.4476 172.4476>}]
 HairSpikesS2 [{shapeType=Capsule, length=5, radius=3, shapeName=HairSpikesS2_capsule, center=<5 0 0>, rotation=<rot 90 179.9975 179.9975>}]
 HairSpikesS3 [{shapeType=Capsule, length=5, radius=3, shapeName=HairSpikesS3_capsule, center=<5 0 0>, rotation=<rot 90 179.8417 179.8418>}]
 HairSpikesS_MSTR [{shapeType=Capsule, length=1, radius=1, shapeName=HairSpikesS_MSTR_capsule}]
 HairSpikesT1 [{shapeType=Capsule, length=5, radius=3, shapeName=HairSpikesT1_capsule, center=<5 0 0>, rotation=<rot 90 179.9951 179.9951>}]
 HairSpikesT_MSTR [{shapeType=Capsule, length=1, radius=1, shapeName=HairSpikesT_MSTR_capsule}]
 HairSpikesU1 [{shapeType=Capsule, length=5, radius=3, shapeName=HairSpikesU1_capsule, center=<5 0 0>, rotation=<rot 90 19.4712 19.4712>}]
 HairSpikesU_MSTR [{shapeType=Capsule, length=1, radius=1, shapeName=HairSpikesU_MSTR_capsule}]
 HairSpikesV1 [{shapeType=Capsule, length=5, radius=3, shapeName=HairSpikesV1_capsule, center=<5 0 0>, rotation=<rot 90 3.0357 3.0357>}]
 HairSpikesV2 [{shapeType=Capsule, length=5, radius=3, shapeName=HairSpikesV2_capsule, center=<5 0 0>, rotation=<rot 90 90 90>}]
 HairSpikesV_MSTR [{shapeType=Capsule, length=1, radius=1, shapeName=HairSpikesV_MSTR_capsule}]
 HairSpikesW1 [{shapeType=Capsule, length=5, radius=3, shapeName=HairSpikesW1_capsule, center=<5 0 0>, rotation=<rot 90 179.9604 179.9604>}]
 HairSpikesW2 [{shapeType=Capsule, length=5, radius=3, shapeName=HairSpikesW2_capsule, center=<5 0 0>, rotation=<rot 90 139.9849 139.9849>}]
 HairSpikesW3 [{shapeType=Capsule, length=5, radius=3, shapeName=HairSpikesW3_capsule, center=<5 0 0>, rotation=<rot 90 -179.7748 -179.7748>}]
 HairSpikesW_MSTR [{shapeType=Capsule, length=1, radius=1, shapeName=HairSpikesW_MSTR_capsule}]
 HairSpikesX [{shapeType=Capsule, length=5, radius=3, shapeName=HairSpikesX_capsule, center=<5 0 0>, rotation=<rot 90 132.1158 132.1158>}]
 HairSpikesX1 [{shapeType=Capsule, length=5, radius=3, shapeName=HairSpikesX1_capsule, center=<5 0 0>, rotation=<rot 90 160.5288 160.5288>}]
 HairSpikesX_MSTR [{shapeType=Capsule, length=1, radius=1, shapeName=HairSpikesX_MSTR_capsule}]
 head [{shapeType=Capsule, length=1, radius=1, shapeName=head_capsule}]
 spine_05 [
  {shapeType=Capsule, length=36.5709, radius=19.4097, shapeName=spine_05_capsule, center=<1.1387 3.7152 0>}
  {shapeType=Capsule, length=39.75, radius=15.25, shapeName=spine_05_capsule, center=<-9.7895 -1.1851 0>}
 ]
}
constraints [
 {bone1Name=HairSpikesA_MSTR, bone2Name=head, swing1LimitDegrees=20, swing2LimitDegrees=20, twistLimitDegrees=0}
 {bone1Name=HairSpikesA1, bone2Name=HairSpikesA_MSTR, swing1LimitDegrees=10, swing2LimitDegrees=5, twistLimitDegrees=0}
 {bone1Name=HairSpikesA2, bone2Name=HairSpikesA1, swing1LimitDegrees=10, swing2LimitDegrees=5, twistLimitDegrees=0}
 {bone1Name=HairSpikesA3, bone2Name=HairSpikesA2, swing1LimitDegrees=10, swing2LimitDegrees=5, twistLimitDegrees=0}
 {bone1Name=HairSpikesB_MSTR, bone2Name=head, swing1LimitDegrees=20, swing2LimitDegrees=20, twistLimitDegrees=0}
 {bone1Name=HairSpikesB1, bone2Name=HairSpikesB_MSTR, swing1LimitDegrees=10, swing2LimitDegrees=5, twistLimitDegrees=0}
 {bone1Name=HairSpikesB2, bone2Name=HairSpikesB1, swing1LimitDegrees=10, swing2LimitDegrees=5, twistLimitDegrees=0}
 {bone1Name=HairSpikesB3, bone2Name=HairSpikesB2, swing1LimitDegrees=10, swing2LimitDegrees=5, twistLimitDegrees=0}
 {bone1Name=HairSpikesC_MSTR, bone2Name=head, swing1LimitDegrees=20, swing2LimitDegrees=20, twistLimitDegrees=0}
 {bone1Name=HairSpikesC1, bone2Name=HairSpikesC_MSTR, swing1LimitDegrees=10, swing2LimitDegrees=5, twistLimitDegrees=0}
 {bone1Name=HairSpikesC2, bone2Name=HairSpikesC1, swing1LimitDegrees=10, swing2LimitDegrees=5, twistLimitDegrees=0}
 {bone1Name=HairSpikesC3, bone2Name=HairSpikesC2, swing1LimitDegrees=10, swing2LimitDegrees=5, twistLimitDegrees=0}
 {bone1Name=HairSpikesD_MSTR, bone2Name=head, swing1LimitDegrees=20, swing2LimitDegrees=20, twistLimitDegrees=0}
 {bone1Name=HairSpikesD1, bone2Name=HairSpikesD_MSTR, swing1LimitDegrees=10, swing2LimitDegrees=5, twistLimitDegrees=0}
 {bone1Name=HairSpikesD2, bone2Name=HairSpikesD1, swing1LimitDegrees=10, swing2LimitDegrees=5, twistLimitDegrees=0}
 {bone1Name=HairSpikesE_MSTR, bone2Name=head, swing1LimitDegrees=20, swing2LimitDegrees=20, twistLimitDegrees=0}
 {bone1Name=HairSpikesE1, bone2Name=HairSpikesE_MSTR, swing1LimitDegrees=5, swing2LimitDegrees=2, twistLimitDegrees=0}
 {bone1Name=HairSpikesF_MSTR, bone2Name=head, swing1LimitDegrees=20, swing2LimitDegrees=20, twistLimitDegrees=0}
 {bone1Name=HairSpikesF1, bone2Name=HairSpikesF_MSTR, swing1LimitDegrees=10, swing2LimitDegrees=5, twistLimitDegrees=0}
 {bone1Name=HairSpikesF2, bone2Name=HairSpikesF1, swing1LimitDegrees=10, swing2LimitDegrees=5, twistLimitDegrees=0}
 {bone1Name=HairSpikesF3, bone2Name=HairSpikesF2, swing1LimitDegrees=10, swing2LimitDegrees=5, twistLimitDegrees=0}
 {bone1Name=HairSpikesG_MSTR, bone2Name=head, swing1LimitDegrees=20, swing2LimitDegrees=20, twistLimitDegrees=0}
 {bone1Name=HairSpikesG1, bone2Name=HairSpikesG_MSTR, swing1LimitDegrees=10, swing2LimitDegrees=5, twistLimitDegrees=0}
 {bone1Name=HairSpikesG2, bone2Name=HairSpikesG1, swing1LimitDegrees=10, swing2LimitDegrees=5, twistLimitDegrees=0}
 {bone1Name=HairSpikesG3, bone2Name=HairSpikesG2, swing1LimitDegrees=10, swing2LimitDegrees=5, twistLimitDegrees=0}
 {bone1Name=HairSpikesG4, bone2Name=HairSpikesG3, swing1LimitDegrees=10, swing2LimitDegrees=5, twistLimitDegrees=0}
 {bone1Name=HairSpikesH_MSTR, bone2Name=head, swing1LimitDegrees=20, swing2LimitDegrees=20, twistLimitDegrees=0}
 {bone1Name=HairSpikesH1, bone2Name=HairSpikesH_MSTR, swing1LimitDegrees=10, swing2LimitDegrees=5, twistLimitDegrees=0}
 {bone1Name=HairSpikesH2, bone2Name=HairSpikesH1, swing1LimitDegrees=10, swing2LimitDegrees=5, twistLimitDegrees=0}
 {bone1Name=HairSpikesH3, bone2Name=HairSpikesH2, swing1LimitDegrees=10, swing2LimitDegrees=5, twistLimitDegrees=0}
 {bone1Name=HairSpikesI_MSTR, bone2Name=head, swing1LimitDegrees=20, swing2LimitDegrees=20, twistLimitDegrees=0}
 {bone1Name=HairSpikesI1, bone2Name=HairSpikesI_MSTR, swing1LimitDegrees=10, swing2LimitDegrees=5, twistLimitDegrees=0}
 {bone1Name=HairSpikesI2, bone2Name=HairSpikesI1, swing1LimitDegrees=10, swing2LimitDegrees=5, twistLimitDegrees=0}
 {bone1Name=HairSpikesI3, bone2Name=HairSpikesI2, swing1LimitDegrees=10, swing2LimitDegrees=5, twistLimitDegrees=0}
 {bone1Name=HairSpikesJ_MSTR, bone2Name=head, swing1LimitDegrees=20, swing2LimitDegrees=20, twistLimitDegrees=0}
 {bone1Name=HairSpikesJ1, bone2Name=HairSpikesJ_MSTR, swing1LimitDegrees=10, swing2LimitDegrees=5, twistLimitDegrees=0}
 {bone1Name=HairSpikesJ2, bone2Name=HairSpikesJ1, swing1LimitDegrees=10, swing2LimitDegrees=5, twistLimitDegrees=0}
 {bone1Name=HairSpikesK_MSTR, bone2Name=head, swing1LimitDegrees=20, swing2LimitDegrees=20, twistLimitDegrees=0}
 {bone1Name=HairSpikesK1, bone2Name=HairSpikesK_MSTR, swing1LimitDegrees=10, swing2LimitDegrees=5, twistLimitDegrees=0}
 {bone1Name=HairSpikesK2, bone2Name=HairSpikesK1, swing1LimitDegrees=10, swing2LimitDegrees=5, twistLimitDegrees=0}
 {bone1Name=HairSpikesK3, bone2Name=HairSpikesK2, swing1LimitDegrees=10, swing2LimitDegrees=5, twistLimitDegrees=0}
 {bone1Name=HairSpikesL_MSTR, bone2Name=head, swing1LimitDegrees=20, swing2LimitDegrees=20, twistLimitDegrees=0}
 {bone1Name=HairSpikesL1, bone2Name=HairSpikesL_MSTR, swing1LimitDegrees=10, swing2LimitDegrees=5, twistLimitDegrees=0}
 {bone1Name=HairSpikesM_MSTR, bone2Name=head, swing1LimitDegrees=20, swing2LimitDegrees=20, twistLimitDegrees=0}
 {bone1Name=HairSpikesM1, bone2Name=HairSpikesM_MSTR, swing1LimitDegrees=5, swing2LimitDegrees=2, twistLimitDegrees=0}
 {bone1Name=HairSpikesN_MSTR, bone2Name=head, swing1LimitDegrees=20, swing2LimitDegrees=20, twistLimitDegrees=0}
 {bone1Name=HairSpikesN1, bone2Name=HairSpikesN_MSTR, swing1LimitDegrees=10, swing2LimitDegrees=5, twistLimitDegrees=0}
 {bone1Name=HairSpikesN2, bone2Name=HairSpikesN1, swing1LimitDegrees=10, swing2LimitDegrees=5, twistLimitDegrees=0}
 {bone1Name=HairSpikesO_MSTR, bone2Name=head, swing1LimitDegrees=20, swing2LimitDegrees=20, twistLimitDegrees=0}
 {bone1Name=HairSpikesO1, bone2Name=HairSpikesO_MSTR, swing1LimitDegrees=10, swing2LimitDegrees=5, twistLimitDegrees=0}
 {bone1Name=HairSpikesO2, bone2Name=HairSpikesO1, swing1LimitDegrees=10, swing2LimitDegrees=5, twistLimitDegrees=0}
 {bone1Name=HairSpikesP_MSTR, bone2Name=head, swing1LimitDegrees=20, swing2LimitDegrees=20, twistLimitDegrees=0}
 {bone1Name=HairSpikesP1, bone2Name=HairSpikesP_MSTR, swing1LimitDegrees=10, swing2LimitDegrees=5, twistLimitDegrees=0}
 {bone1Name=HairSpikesQ_MSTR, bone2Name=head, swing1LimitDegrees=20, swing2LimitDegrees=20, twistLimitDegrees=0}
 {bone1Name=HairSpikesQ1, bone2Name=HairSpikesQ_MSTR, swing1LimitDegrees=10, swing2LimitDegrees=5, twistLimitDegrees=5}
 {bone1Name=HairSpikesQ2, bone2Name=HairSpikesQ1, swing1LimitDegrees=10, swing2LimitDegrees=5, twistLimitDegrees=5}
 {bone1Name=HairSpikesQ3, bone2Name=HairSpikesQ2, swing1LimitDegrees=10, swing2LimitDegrees=5, twistLimitDegrees=5}
 {bone1Name=HairSpikesR_MSTR, bone2Name=head, swing1LimitDegrees=20, swing2LimitDegrees=20, twistLimitDegrees=0}
 {bone1Name=HairSpikesR1, bone2Name=HairSpikesR_MSTR, swing1LimitDegrees=10, swing2LimitDegrees=5, twistLimitDegrees=0}
 {bone1Name=HairSpikesR2, bone2Name=HairSpikesR1, swing1LimitDegrees=10, swing2LimitDegrees=5, twistLimitDegrees=0}
 {bone1Name=HairSpikesR3, bone2Name=HairSpikesR2, swing1LimitDegrees=10, swing2LimitDegrees=5, twistLimitDegrees=0}
 {bone1Name=HairSpikesS_MSTR, bone2Name=head, swing1LimitDegrees=20, swing2LimitDegrees=20, twistLimitDegrees=0}
 {bone1Name=HairSpikesS1, bone2Name=HairSpikesS_MSTR, swing1LimitDegrees=10, swing2LimitDegrees=5, twistLimitDegrees=0}
 {bone1Name=HairSpikesS2, bone2Name=HairSpikesS1, swing1LimitDegrees=10, swing2LimitDegrees=5, twistLimitDegrees=0}
 {bone1Name=HairSpikesS3, bone2Name=HairSpikesS2, swing1LimitDegrees=10, swing2LimitDegrees=5, twistLimitDegrees=0}
 {bone1Name=HairSpikesT_MSTR, bone2Name=head, swing1LimitDegrees=20, swing2LimitDegrees=20, twistLimitDegrees=0}
 {bone1Name=HairSpikesT1, bone2Name=HairSpikesT_MSTR, swing1LimitDegrees=10, swing2LimitDegrees=5, twistLimitDegrees=0}
 {bone1Name=HairSpikesU_MSTR, bone2Name=head, swing1LimitDegrees=20, swing2LimitDegrees=20, twistLimitDegrees=0}
 {bone1Name=HairSpikesU1, bone2Name=HairSpikesU_MSTR, swing1LimitDegrees=5, swing2LimitDegrees=2, twistLimitDegrees=0}
 {bone1Name=HairSpikesV_MSTR, bone2Name=head, swing1LimitDegrees=20, swing2LimitDegrees=20, twistLimitDegrees=0}
 {bone1Name=HairSpikesV1, bone2Name=HairSpikesV_MSTR, swing1LimitDegrees=5, swing2LimitDegrees=2, twistLimitDegrees=0}
 {bone1Name=HairSpikesV2, bone2Name=HairSpikesV1, swing1LimitDegrees=5, swing2LimitDegrees=2, twistLimitDegrees=0}
 {bone1Name=HairSpikesW_MSTR, bone2Name=head, swing1LimitDegrees=20, swing2LimitDegrees=20, twistLimitDegrees=0}
 {bone1Name=HairSpikesW1, bone2Name=HairSpikesW_MSTR, swing1LimitDegrees=10, swing2LimitDegrees=5, twistLimitDegrees=0}
 {bone1Name=HairSpikesW2, bone2Name=HairSpikesW1, swing1LimitDegrees=10, swing2LimitDegrees=5, twistLimitDegrees=0}
 {bone1Name=HairSpikesW3, bone2Name=HairSpikesW2, swing1LimitDegrees=10, swing2LimitDegrees=5, twistLimitDegrees=0}
 {bone1Name=HairSpikesX_MSTR, bone2Name=head, swing1LimitDegrees=20, swing2LimitDegrees=20, twistLimitDegrees=0}
 {bone1Name=HairSpikesX1, bone2Name=HairSpikesX_MSTR, swing1LimitDegrees=10, swing2LimitDegrees=5, twistLimitDegrees=0}
 {bone1Name=HairSpikesX, bone2Name=HairSpikesX1, swing1LimitDegrees=10, swing2LimitDegrees=5, twistLimitDegrees=0}
 {
  bone1Name=head
  bone2Name=spine_05
  swing1LimitDegrees=45
  swing1Motion=Free
  swing2LimitDegrees=45
  swing2Motion=Free
  twistLimitDegrees=45
  twistMotion=Free
 }
 {
  bone1Name=HairSpikesS3
  bone2Name=HairSpikesW3
  swing1LimitDegrees=45
  swing1Motion=Free
  swing2LimitDegrees=45
  swing2Motion=Free
  twistLimitDegrees=45
  twistMotion=Free
 }
 {
  bone1Name=HairSpikesQ3
  bone2Name=HairSpikesG4
  swing1LimitDegrees=45
  swing1Motion=Free
  swing2LimitDegrees=45
  swing2Motion=Free
  twistLimitDegrees=45
  twistMotion=Free
 }
]
