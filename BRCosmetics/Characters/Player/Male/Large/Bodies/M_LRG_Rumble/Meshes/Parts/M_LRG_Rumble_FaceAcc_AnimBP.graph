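# Post-process style AnimBP for the Rumble face accessory mesh that poses and simulates hair spikes on top of the copied parent body pose.
# Two branches, chosen by backbling presence, each constrain hairSpikeMain to head/neck, then run head-space RigidBody with its own physics asset.
# Additive bone corrections surround the simulation, weighted by backward head tilt and movement or frontend states from EventGraph.
# Long backbling, detected once after customization loads, adds final cape-clearance offsets.
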AnimBlueprint /BRCosmetics/Characters/Player/Male/Large/Bodies/M_LRG_Rumble/Meshes/Parts/M_LRG_Rumble_FaceAcc_AnimBP
target_skeleton=/Game/Characters/Player/Male/Large/Bodies/M_LRG_Rumble/Meshes/Parts/M_LRG_Rumble_FaceAcc_Skeleton
graphs [
 AnimGraph {nodes [
   n1 = Get headRotBack() -> (headRotBack:double)
   n2 = "Transform (Modify) Bone - Bone: hairSpikeLowerMain"(Alpha=@n1, Translation="0,0,0", Rotation="0,0,0", Scale="1,1,1") {
    rotationMode=BMM_Additive
    rotationSpace=BCS_BoneSpace
    scaleMode=BMM_Replace
    scaleSpace=BCS_BoneSpace
    translationMode=BMM_Additive
    translationSpace=BCS_BoneSpace
    alphaBoolBlend {blendOption=Linear}
    boneToModify {boneName=hairSpikeLowerMain}
   }
   "Transform (Modify) Bone - Bone: hairSpikeMain"(ComponentPose=@n2, Alpha=@n1, Translation="0,0,0", Rotation="0,0,-15", Scale="1,1,1") {
    rotationMode=BMM_Additive
    rotationSpace=BCS_BoneSpace
    scaleMode=BMM_Replace
    scaleSpace=BCS_BoneSpace
    translationMode=BMM_Additive
    translationSpace=BCS_BoneSpace
    alphaBoolBlend {blendOption=Linear}
    boneToModify {boneName=hairSpikeMain}
    rotation {roll=-15}
   }
   n3 = "Constraint - hairSpikeMain (RR)"!(
    ComponentPose=Local To Component(LocalPose=Copy Pose From Mesh() {bCopyCurves=true, bUseAttachedParent=true})
    Alpha=1
    ConstraintWeights_0=0.75
    ConstraintWeights_1:float=0.25
   ) -> (Pose) {
    alphaBoolBlend {blendOption=Linear}
    boneToModify {boneName=hairSpikeMain}
    constraintSetup [{transformType=Rotation, targetBone {boneName=head}}, {transformType=Rotation, targetBone {boneName=neck_02}}]
    constraintWeights [1,1]
   }
   n4 = "Transform (Modify) Bone - Bone: hairSpikeMain"(ComponentPose=@n3, Alpha=@n1, Translation="0,0,0", Rotation="-5,5,-20", Scale="1,1,1") {
    rotationMode=BMM_Additive
    rotationSpace=BCS_BoneSpace
    scaleMode=BMM_Replace
    scaleSpace=BCS_BoneSpace
    translationMode=BMM_Additive
    translationSpace=BCS_BoneSpace
    alphaBoolBlend {blendOption=Linear}
    boneToModify {boneName=hairSpikeMain}
    rotation=<rot -5 5 -20>
   }
   n5 = Get isInFrontendNoEmoting() -> (isInFrontendNoEmoting:double)
   n6 = "Transform (Modify) Bone - Bone: HairSpikesS_MSTR"(ComponentPose=@n4, Alpha=@n5, Translation="0,0,0", Rotation="10,15,0", Scale="1,1,1") {
    rotationMode=BMM_Additive
    rotationSpace=BCS_BoneSpace
    scaleMode=BMM_Replace
    scaleSpace=BCS_BoneSpace
    translationMode=BMM_Additive
    translationSpace=BCS_BoneSpace
    alphaBoolBlend {blendOption=Linear}
    boneToModify {boneName=HairSpikesS_MSTR}
    rotation {pitch=10, yaw=15}
   }
   n7 = "Transform (Modify) Bone - Bone: hairSpikeMain"(ComponentPose=@n6, Alpha=Get Is Crouching() -> (Is Crouching:double), Translation="0,0,0", Rotation="0,5,0", Scale="1,1,1") {
    rotationMode=BMM_Additive
    rotationSpace=BCS_BoneSpace
    scaleMode=BMM_Replace
    scaleSpace=BCS_BoneSpace
    translationMode=BMM_Additive
    translationSpace=BCS_BoneSpace
    alphaBoolBlend {blendOption=Linear}
    boneToModify {boneName=hairSpikeMain}
    rotation {yaw=5}
   }
   n8 = Get GravityOverride()
   n9 = RigidBody!(ComponentPose=@n7, Alpha, OverrideWorldGravity:Vector=@n8, ExternalForce) -> (Pose) {
    bClampLinearTranslationLimitToRefPose=true
    bOverrideWorldGravity=true
    overridePhysicsAsset=/Game/Characters/Player/Male/Large/Bodies/M_LRG_Rumble/Meshes/Parts/M_LRG_Rumble_FaceAcc_Physics
    simulationSpace=BaseBoneSpace
    alphaBoolBlend {blendOption=Linear}
    baseBoneRef {boneName=head}
    componentAppliedLinearAccClamp=<3000 3000 3000>
    componentLinearAccScale=<0.35 0.3 0.3>
    componentLinearVelScale=<0.35 0.3 0.3>
    overrideWorldGravity {z=-980}
    simSpaceSettings {
     maxAngularAcceleration=1000
     maxAngularVelocity=3
     maxLinearAcceleration=700
     maxLinearVelocity=500
     velocityScaleZ=0
     worldAlpha=1
     externalLinearDragV=<0.2 0.2>
    }
   }
   n10 = "Transform (Modify) Bone - Bone: hairSpikeLowerMain"(ComponentPose=@n9, Alpha=@n5, Translation="0,0,0", Rotation="0.000,-6,-0.000", Scale="1,1,1") {
    rotationMode=BMM_Additive
    rotationSpace=BCS_BoneSpace
    scaleMode=BMM_Replace
    scaleSpace=BCS_BoneSpace
    translationMode=BMM_Additive
    translationSpace=BCS_BoneSpace
    alphaBoolBlend {blendOption=Linear}
    boneToModify {boneName=hairSpikeLowerMain}
    rotation {yaw=-6}
   }
   n11 = Get headRotBack()
   n12 = "Transform (Modify) Bone - Bone: hairSpikeMain"(ComponentPose=@n10, Alpha=@n11, Translation="0,-0.5,0", Rotation="0,10,15", Scale="1,1,1") {
    rotationMode=BMM_Additive
    rotationSpace=BCS_BoneSpace
    scaleMode=BMM_Replace
    scaleSpace=BCS_BoneSpace
    translationMode=BMM_Additive
    translationSpace=BCS_BoneSpace
    alphaBoolBlend {blendOption=Linear}
    boneToModify {boneName=hairSpikeMain}
    rotation {roll=15, yaw=10}
    translation {y=-0.5}
   }
   n13 = "Transform (Modify) Bone - Bone: HairSpikesH_MSTR"(ComponentPose=@n12, Alpha=@n11, Translation="0,0,0", Rotation="0,10,0", Scale="1,1,1") {
    rotationMode=BMM_Additive
    rotationSpace=BCS_BoneSpace
    scaleMode=BMM_Replace
    scaleSpace=BCS_BoneSpace
    translationMode=BMM_Additive
    translationSpace=BCS_BoneSpace
    alphaBoolBlend {blendOption=Linear}
    boneToModify {boneName=HairSpikesH_MSTR}
    rotation {yaw=10}
   }
   n14 = "Transform (Modify) Bone - Bone: HairSpikesK_MSTR"(ComponentPose=@n13, Alpha=@n11, Translation="0,0,0", Rotation="-10,15,0", Scale="1,1,1") {
    rotationMode=BMM_Additive
    rotationSpace=BCS_BoneSpace
    scaleMode=BMM_Replace
    scaleSpace=BCS_BoneSpace
    translationMode=BMM_Additive
    translationSpace=BCS_BoneSpace
    alphaBoolBlend {blendOption=Linear}
    boneToModify {boneName=HairSpikesK_MSTR}
    rotation {pitch=-10, yaw=15}
   }
   n15 = "Transform (Modify) Bone - Bone: HairSpikesC_MSTR"(ComponentPose=@n14, Alpha=@n11, Translation="0,0,0", Rotation="0,15,0", Scale="1,1,1") {
    rotationMode=BMM_Additive
    rotationSpace=BCS_BoneSpace
    scaleMode=BMM_Replace
    scaleSpace=BCS_BoneSpace
    translationMode=BMM_Additive
    translationSpace=BCS_BoneSpace
    alphaBoolBlend {blendOption=Linear}
    boneToModify {boneName=HairSpikesC_MSTR}
    rotation {yaw=15}
   }
   n16 = "Transform (Modify) Bone - Bone: HairSpikesB_MSTR"(ComponentPose=@n15, Alpha=@n11, Translation="0,0,0", Rotation="0,5,0", Scale="1,1,1") {
    rotationMode=BMM_Additive
    rotationSpace=BCS_BoneSpace
    scaleMode=BMM_Replace
    scaleSpace=BCS_BoneSpace
    translationMode=BMM_Additive
    translationSpace=BCS_BoneSpace
    alphaBoolBlend {blendOption=Linear}
    boneToModify {boneName=HairSpikesB_MSTR}
    rotation {yaw=5}
   }
   n17 = "Transform (Modify) Bone - Bone: HairSpikesC_MSTR"(ComponentPose=@n16, Alpha=@n11, Translation="0,0,0", Rotation="5,10,0", Scale="1,1,1") {
    rotationMode=BMM_Additive
    rotationSpace=BCS_BoneSpace
    scaleMode=BMM_Replace
    scaleSpace=BCS_BoneSpace
    translationMode=BMM_Additive
    translationSpace=BCS_BoneSpace
    alphaBoolBlend {blendOption=Linear}
    boneToModify {boneName=HairSpikesC_MSTR}
    rotation {pitch=5, yaw=10}
   }
   n18 = "Transform (Modify) Bone - Bone: HairSpikesI_MSTR"(ComponentPose=@n17, Alpha=@n11, Translation="0,0,0", Rotation="0,30,0", Scale="1,1,1") {
    rotationMode=BMM_Additive
    rotationSpace=BCS_BoneSpace
    scaleMode=BMM_Replace
    scaleSpace=BCS_BoneSpace
    translationMode=BMM_Additive
    translationSpace=BCS_BoneSpace
    alphaBoolBlend {blendOption=Linear}
    boneToModify {boneName=HairSpikesI_MSTR}
    rotation {yaw=30}
   }
   n19 = "Transform (Modify) Bone - Bone: HairSpikesS_MSTR"(ComponentPose=@n18, Alpha=@n11, Translation="0,3,0", Rotation="0,0,0", Scale="1,1,1") {
    rotationMode=BMM_Additive
    rotationSpace=BCS_BoneSpace
    scaleMode=BMM_Replace
    scaleSpace=BCS_BoneSpace
    translationMode=BMM_Additive
    translationSpace=BCS_BoneSpace
    alphaBoolBlend {blendOption=Linear}
    boneToModify {boneName=HairSpikesS_MSTR}
    translation {y=3}
   }
   n20 = "Transform (Modify) Bone - Bone: HairSpikesQ_MSTR"(ComponentPose=@n19, Alpha=@n11, Translation="0,0,0", Rotation="15,0,0", Scale="1,1,1") {
    rotationMode=BMM_Additive
    rotationSpace=BCS_BoneSpace
    scaleMode=BMM_Replace
    scaleSpace=BCS_BoneSpace
    translationMode=BMM_Additive
    translationSpace=BCS_BoneSpace
    alphaBoolBlend {blendOption=Linear}
    boneToModify {boneName=HairSpikesQ_MSTR}
    rotation {pitch=15}
   }
   n21 = "Transform (Modify) Bone - Bone: HairSpikesR_MSTR"(ComponentPose=@n20, Alpha=@n11, Translation="0,0,0", Rotation="0,20,0", Scale="1,1,1") {
    rotationMode=BMM_Additive
    rotationSpace=BCS_BoneSpace
    scaleMode=BMM_Replace
    scaleSpace=BCS_BoneSpace
    translationMode=BMM_Additive
    translationSpace=BCS_BoneSpace
    alphaBoolBlend {blendOption=Linear}
    boneToModify {boneName=HairSpikesR_MSTR}
    rotation {yaw=20}
   }
   n22 = "Transform (Modify) Bone - Bone: HairSpikesR_MSTR"(ComponentPose=@n21, Alpha=1, Translation="0,0,0", Rotation="0,10,0", Scale="1,1,1") {
    rotationMode=BMM_Additive
    rotationSpace=BCS_BoneSpace
    scaleMode=BMM_Replace
    scaleSpace=BCS_BoneSpace
    translationMode=BMM_Additive
    translationSpace=BCS_BoneSpace
    alphaBoolBlend {blendOption=Linear}
    boneToModify {boneName=HairSpikesR_MSTR}
    rotation {yaw=10}
   }
   n23 = "Transform (Modify) Bone - Bone: HairSpikesT_MSTR"(ComponentPose=@n22, Alpha=@n11, Translation="0,0,0", Rotation="0,5,0", Scale="1,1,1") {
    rotationMode=BMM_Additive
    rotationSpace=BCS_BoneSpace
    scaleMode=BMM_Replace
    scaleSpace=BCS_BoneSpace
    translationMode=BMM_Additive
    translationSpace=BCS_BoneSpace
    alphaBoolBlend {blendOption=Linear}
    boneToModify {boneName=HairSpikesT_MSTR}
    rotation {yaw=5}
   }
   n24 = "Transform (Modify) Bone - Bone: HairSpikesF_MSTR"(ComponentPose=@n23, Alpha=@n11, Translation="0,0,0", Rotation="-5,0,0", Scale="1,1,1") {
    rotationMode=BMM_Additive
    rotationSpace=BCS_BoneSpace
    scaleMode=BMM_Replace
    scaleSpace=BCS_BoneSpace
    translationMode=BMM_Additive
    translationSpace=BCS_BoneSpace
    alphaBoolBlend {blendOption=Linear}
    boneToModify {boneName=HairSpikesF_MSTR}
    rotation {pitch=-5}
   }
   n25 = "Transform (Modify) Bone - Bone: hairSpikeMain"(ComponentPose=@n24, Alpha=Get Is Crouching(), Translation="0,-1,0", Rotation="0,10,0", Scale="1,1,1") {
    rotationMode=BMM_Additive
    rotationSpace=BCS_BoneSpace
    scaleMode=BMM_Replace
    scaleSpace=BCS_BoneSpace
    translationMode=BMM_Additive
    translationSpace=BCS_BoneSpace
    alphaBoolBlend {blendOption=Linear}
    boneToModify {boneName=hairSpikeMain}
    rotation {yaw=10}
    translation {y=-1}
   }
   n26 = "Transform (Modify) Bone - Bone: hairSpikeMain"(ComponentPose=@n25, Alpha=Get Is Sky Diving() -> (Is Sky Diving:double), Translation="0,0,0", Rotation="0,15,0", Scale="1,1,1") {
    rotationMode=BMM_Additive
    rotationSpace=BCS_BoneSpace
    scaleMode=BMM_Replace
    scaleSpace=BCS_BoneSpace
    translationMode=BMM_Additive
    translationSpace=BCS_BoneSpace
    alphaBoolBlend {blendOption=Linear}
    boneToModify {boneName=hairSpikeMain}
    rotation {yaw=15}
   }
   n27 = "Transform (Modify) Bone - Bone: hairSpikeMain"(ComponentPose=@n26, Alpha=Get Is Moving 2D() -> (Is Moving 2D:double), Translation="0,0,0", Rotation="0,3,0", Scale="1,1,1") {
    rotationMode=BMM_Additive
    rotationSpace=BCS_BoneSpace
    scaleMode=BMM_Replace
    scaleSpace=BCS_BoneSpace
    translationMode=BMM_Additive
    translationSpace=BCS_BoneSpace
    alphaBoolBlend {blendOption=Linear}
    boneToModify {boneName=hairSpikeMain}
    rotation {yaw=3}
   }
   n28 = Component To Local(ComponentPose=@n27)
   n29 = "Constraint - hairSpikeMain (RR)"(
    ComponentPose=Local To Component(LocalPose=Copy Pose From Mesh() {bCopyCurves=true, bUseAttachedParent=true})
    Alpha=1
    ConstraintWeights_0=0.5
    ConstraintWeights_1=0.5
   ) {
    alphaBoolBlend {blendOption=Linear}
    boneToModify {boneName=hairSpikeMain}
    constraintSetup [{transformType=Rotation, targetBone {boneName=head}}, {transformType=Rotation, targetBone {boneName=neck_02}}]
    constraintWeights [1,1]
   }
   n30 = "Transform (Modify) Bone - Bone: hairSpikeMain"(ComponentPose=@n29, Alpha=@n1, Translation="0,0,0", Rotation="-5,5,-20", Scale="1,1,1") {
    rotationMode=BMM_Additive
    rotationSpace=BCS_BoneSpace
    scaleMode=BMM_Replace
    scaleSpace=BCS_BoneSpace
    translationMode=BMM_Additive
    translationSpace=BCS_BoneSpace
    alphaBoolBlend {blendOption=Linear}
    boneToModify {boneName=hairSpikeMain}
    rotation=<rot -5 5 -20>
   }
   n31 = "Transform (Modify) Bone - Bone: hairSpikeMain"(ComponentPose=@n30, Alpha=@n1, Translation="0,0,0", Rotation="0,5,0", Scale="1,1,1") {
    rotationMode=BMM_Additive
    rotationSpace=BCS_BoneSpace
    scaleMode=BMM_Replace
    scaleSpace=BCS_BoneSpace
    translationMode=BMM_Additive
    translationSpace=BCS_BoneSpace
    alphaBoolBlend {blendOption=Linear}
    boneToModify {boneName=hairSpikeMain}
    rotation {yaw=5}
   }
   n32 = RigidBody(ComponentPose=@n31, OverrideWorldGravity=@n8) {
    bClampLinearTranslationLimitToRefPose=true
    bOverrideWorldGravity=true
    overridePhysicsAsset=/Game/Characters/Player/Male/Large/Bodies/M_LRG_Rumble/Meshes/Parts/M_LRG_Rumble_FaceAcc_PhysicsBackPack
    simulationSpace=BaseBoneSpace
    alphaBoolBlend {blendOption=Linear}
    baseBoneRef {boneName=head}
    componentAppliedLinearAccClamp=<3000 3000 3000>
    componentLinearAccScale=<0.35 0.3 0.3>
    componentLinearVelScale=<0.35 0.3 0.3>
    overrideWorldGravity {z=-980}
    simSpaceSettings {
     maxAngularAcceleration=1000
     maxAngularVelocity=3
     maxLinearAcceleration=700
     maxLinearVelocity=500
     velocityScaleZ=0
     worldAlpha=1
     externalLinearDragV=<0.2 0.2>
    }
   }
   n33 = "Transform (Modify) Bone - Bone: hairSpikeMain"(ComponentPose=@n32, Alpha=@n11, Translation="0,-0.5,0", Rotation="0,10,15", Scale="1,1,1") {
    rotationMode=BMM_Additive
    rotationSpace=BCS_BoneSpace
    scaleMode=BMM_Replace
    scaleSpace=BCS_BoneSpace
    translationMode=BMM_Additive
    translationSpace=BCS_BoneSpace
    alphaBoolBlend {blendOption=Linear}
    boneToModify {boneName=hairSpikeMain}
    rotation {roll=15, yaw=10}
    translation {y=-0.5}
   }
   n34 = "Transform (Modify) Bone - Bone: HairSpikesC_MSTR"(ComponentPose=@n33, Alpha=@n11, Translation="0,0,0", Rotation="0,7,0", Scale="1,1,1") {
    rotationMode=BMM_Additive
    rotationSpace=BCS_BoneSpace
    scaleMode=BMM_Replace
    scaleSpace=BCS_BoneSpace
    translationMode=BMM_Additive
    translationSpace=BCS_BoneSpace
    alphaBoolBlend {blendOption=Linear}
    boneToModify {boneName=HairSpikesC_MSTR}
    rotation {yaw=7}
   }
   n35 = "Transform (Modify) Bone - Bone: HairSpikesI3"(ComponentPose=@n34, Alpha=@n11, Translation="0,0,0", Rotation="0,-15,0", Scale="1,1,1") {
    rotationMode=BMM_Additive
    rotationSpace=BCS_BoneSpace
    scaleMode=BMM_Replace
    scaleSpace=BCS_BoneSpace
    translationMode=BMM_Additive
    translationSpace=BCS_BoneSpace
    alphaBoolBlend {blendOption=Linear}
    boneToModify {boneName=HairSpikesI3}
    rotation {yaw=-15}
   }
   n36 = "Transform (Modify) Bone - Bone: HairSpikesG3"(ComponentPose=@n35, Alpha=@n11, Translation="0,0,0", Rotation="0,-10,0", Scale="1,1,1") {
    rotationMode=BMM_Additive
    rotationSpace=BCS_BoneSpace
    scaleMode=BMM_Replace
    scaleSpace=BCS_BoneSpace
    translationMode=BMM_Additive
    translationSpace=BCS_BoneSpace
    alphaBoolBlend {blendOption=Linear}
    boneToModify {boneName=HairSpikesG3}
    rotation {yaw=-10}
   }
   n37 = "Transform (Modify) Bone - Bone: HairSpikesQ2"(ComponentPose=@n36, Alpha=@n11, Translation="0,0,0", Rotation="0,0,0", Scale="1,1,1") {
    rotationMode=BMM_Additive
    rotationSpace=BCS_BoneSpace
    scaleMode=BMM_Replace
    scaleSpace=BCS_BoneSpace
    translationMode=BMM_Additive
    translationSpace=BCS_BoneSpace
    alphaBoolBlend {blendOption=Linear}
    boneToModify {boneName=HairSpikesQ2}
   }
   n38 = "Transform (Modify) Bone - Bone: HairSpikesR_MSTR"(ComponentPose=@n37, Alpha=@n11, Translation="0,0,0", Rotation="0,15,0", Scale="1,1,1") {
    rotationMode=BMM_Additive
    rotationSpace=BCS_BoneSpace
    scaleMode=BMM_Replace
    scaleSpace=BCS_BoneSpace
    translationMode=BMM_Additive
    translationSpace=BCS_BoneSpace
    alphaBoolBlend {blendOption=Linear}
    boneToModify {boneName=HairSpikesR_MSTR}
    rotation {yaw=15}
   }
   n39 = "Transform (Modify) Bone - Bone: hairSpikeMain"(ComponentPose=@n38, Alpha=Get Is Sky Diving(), Translation="0,0,0", Rotation="0,0,0", Scale="1,1,1") {
    rotationMode=BMM_Additive
    rotationSpace=BCS_BoneSpace
    scaleMode=BMM_Replace
    scaleSpace=BCS_BoneSpace
    translationMode=BMM_Additive
    translationSpace=BCS_BoneSpace
    alphaBoolBlend {blendOption=Linear}
    boneToModify {boneName=hairSpikeMain}
   }
   n40 = "Transform (Modify) Bone - Bone: hairSpikeMain"(ComponentPose=@n39, Alpha=0, Translation="0,0,0", Rotation="0,3,0", Scale="1,1,1") {
    rotationMode=BMM_Additive
    rotationSpace=BCS_BoneSpace
    scaleMode=BMM_Replace
    scaleSpace=BCS_BoneSpace
    translationMode=BMM_Additive
    translationSpace=BCS_BoneSpace
    alphaBoolBlend {blendOption=Linear}
    boneToModify {boneName=hairSpikeMain}
    rotation {yaw=3}
   }
   n41 = Component To Local(ComponentPose=@n40)
   n42 = Get isWearBackbling() -> (isWearBackbling:bool)
   n43 = Get isWearCape() -> (isWearCape:bool)
   n44 = "Transform (Modify) Bone - Bone: hairSpikeMain"!(
    ComponentPose=Local To Component(LocalPose=Blend Poses by bool(BlendPose_0=@n41, BlendPose_1=@n28, BlendTime_0=0.3, BlendTime_1=0.3, bActiveValue=@n42) {blendType=Linear})
    bAlphaBoolEnabled:bool=@n43
    Translation="0,-7,0"
    Rotation="0,0,15"
    Scale="1,1,1"
   ) -> (Pose) {
    alphaInputType=Bool
    rotationMode=BMM_Additive
    rotationSpace=BCS_BoneSpace
    scaleMode=BMM_Replace
    scaleSpace=BCS_BoneSpace
    translationMode=BMM_Additive
    translationSpace=BCS_BoneSpace
    alphaBoolBlend {blendOption=Linear}
    boneToModify {boneName=hairSpikeMain}
    rotation {roll=15}
    translation {y=-7}
   }
   n45 = "Transform (Modify) Bone - Bone: HairSpikesG1"!(ComponentPose=@n44, bAlphaBoolEnabled:bool=@n43, Translation="0,0,0", Rotation="0,0,0", Scale="0.9,1,1") -> (Pose) {
    alphaInputType=Bool
    rotationMode=BMM_Additive
    rotationSpace=BCS_BoneSpace
    scaleMode=BMM_Replace
    scaleSpace=BCS_BoneSpace
    translationMode=BMM_Additive
    translationSpace=BCS_BoneSpace
    alphaBoolBlend {blendOption=Linear}
    boneToModify {boneName=HairSpikesG1}
    scale {x=0.9}
   }
   n46 = "Transform (Modify) Bone - Bone: HairSpikesG2"!(ComponentPose=@n45, bAlphaBoolEnabled:bool=@n43, Translation="0,0,0", Rotation="0,0,0", Scale="0.9,1,1") -> (Pose) {
    alphaInputType=Bool
    rotationMode=BMM_Additive
    rotationSpace=BCS_BoneSpace
    scaleMode=BMM_Replace
    scaleSpace=BCS_BoneSpace
    translationMode=BMM_Additive
    translationSpace=BCS_BoneSpace
    alphaBoolBlend {blendOption=Linear}
    boneToModify {boneName=HairSpikesG2}
    scale {x=0.9}
   }
   n47 = "Transform (Modify) Bone - Bone: HairSpikesG3"!(ComponentPose=@n46, bAlphaBoolEnabled:bool=@n43, Translation="0,0,0", Rotation="0,10,0", Scale="0.9,1,1") -> (Pose) {
    alphaInputType=Bool
    rotationMode=BMM_Additive
    rotationSpace=BCS_BoneSpace
    scaleMode=BMM_Replace
    scaleSpace=BCS_BoneSpace
    translationMode=BMM_Additive
    translationSpace=BCS_BoneSpace
    alphaBoolBlend {blendOption=Linear}
    boneToModify {boneName=HairSpikesG3}
    rotation {yaw=10}
    scale {x=0.9}
   }
   Output Pose(Result=Component To Local(ComponentPose=@n47)) {layerGroup=None}
   Get Is Moving 2D()
  ]}
 EventGraph {nodes [
   Event BlueprintUpdateAnimation()
   > n1 = Sequence!(execute) -> (then_0, then_1, then_2:Exec, then_3:Exec)
   > Set Is Crouching(Is Crouching:double="To Float (Boolean)"(InBool=Get bIsCrouching())) -> (Output_Get:double)
   n2 = Gravity Override State Function DEPRECATED(
    execute=@n1.then_1
    CharacterPartAnimInstance=Self-Reference()
    GameTimeInSeconds=GetGameTimeInSeconds()
    GravityOverrideParams=Get Gravity Override Params()
   )
   > Set GravityOverride(GravityOverride=@n2)
   > Set headRotBack(headRotBack:double=GetAlphaFromReferenceTransformDelta(
     SkeletalMeshComponent=GetOwnerSkeletalMeshComponent()
     BoneName=head
     BaseName=spine_05
     InputRange=(X=-30,Y=10)
     OutputRange=(X=1,Y=0)
     TwistAxis="1,0,0"
    )) -> (Output_Get:double)
   > Set Is Targeting(Is Targeting:double="To Float (Boolean)"(InBool=Get bIsTargeting())) -> (Output_Get:double)
   > Set isTargetingHeadBack(isTargetingHeadBack:double=float * float(A=Get Is Targeting() -> (Is Targeting:double), B=Get headRotBack()) {is=title K2Node_PromotableOperator}) -> (Output_Get:double)
   > Set Is Sky Diving(Is Sky Diving:double="To Float (Boolean)"(InBool=Get bIsSkyDiving())) -> (Output_Get:double)
   > Set Is Moving 2D(Is Moving 2D:double="To Float (Boolean)"(InBool=Get bIsMoving2D())) -> (Output_Get:double)
   n3 = Set Is Wear BackblingFloat(execute=@n1.then_2, Is Wear BackblingFloat:double="To Float (Boolean)"(InBool=Get isWearBackbling())) -> (Output_Get:double)
   > Set headRotBackBB(headRotBackBB:double=float * float(A=@n3, B=Get headRotBack())) -> (Output_Get:double)
   n4 = NOT Boolean(A=Get bIsPlayingEmote())
   n5 = Get bIsInFrontEnd()
   Set isInFrontendNoEmoting(execute=@n1.then_3, isInFrontendNoEmoting:double="To Float (Boolean)"(InBool=AND Boolean(A=@n5, B=@n4))) -> (Output_Get:double)
   Event OnCharacterCustomizationLoaded()
   > Do Once()
   n6 = CheckForBackbling()
   > Set isWearBackbling(isWearBackbling:bool=@n6) -> (Output_Get:bool)
   > Set isWearCape(isWearCape:bool=@n6.bIsBackblingLong) -> (Output_Get:bool)
   TryGetPawnOwner()
  ]}
]
variables [
 Gravity Override Params {type=GravityOverrideParamsStruct}
 GravityOverride {type=Vector}
 headRotBack {type=number}
 headRotBackBB {type=number}
 Is Crouching {type=number}
 Is Moving 2D {type=number}
 Is Sky Diving {type=number}
 Is Targeting {type=number}
 Is Targeting_0 {type=number}
 Is Wear BackblingFloat {type=number}
 isInFrontendNoEmoting {type=number}
 isNotWearingBB {type=number}
 isTargetingHeadBack {type=number}
 isWearBackbling {type=boolean}
 isWearCape {type=boolean}
 OneFloat {type=number}
 OnFloatAgain {type=number}
]
